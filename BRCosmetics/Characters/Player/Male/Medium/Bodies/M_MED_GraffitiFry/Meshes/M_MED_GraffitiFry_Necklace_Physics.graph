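# Physics asset for the necklace on the GraffitiFry medium male player cosmetic, giving its dynamic bones collision and jointed motion.
# A large capsule on spine_05 covers the upper torso, and a tiny capsule on base_necklace roots a two-capsule chain along dyn_necklace_1 and dyn_necklace_2.
# The base-to-spine joint leaves swing and twist free, while the two chain joints have tight angular limits that keep the swing small.
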PhysicsAsset /BRCosmetics/Characters/Player/Male/Medium/Bodies/M_MED_GraffitiFry/Meshes/M_MED_GraffitiFry_Necklace_Physics
bodies {
 base_necklace [{shapeType=Capsule, length=0.505, radius=0.505, shapeName=base_necklace_capsule, rotation=<rot 0 -90 0>}]
 dyn_necklace_1 [{shapeType=Capsule, length=6, radius=2, shapeName=dyn_necklace_2_capsule, center=<4 1.3 0>, rotation=<rot 90 0 0>}]
 dyn_necklace_2 [{shapeType=Capsule, length=6, radius=2.1, shapeName=dyn_necklace_2_capsule, center=<2 1.5 0>, rotation=<rot 90 0 -5>}]
 spine_05 [{shapeType=Capsule, length=11, radius=11, shapeName=spine_05_capsule, center=<-2.8 0.2 0>}]
}
constraints [
 {
  bone1Name=base_necklace
  bone2Name=spine_05
  swing1LimitDegrees=45
  swing1Motion=Free
  swing2LimitDegrees=45
  swing2Motion=Free
  twistLimitDegrees=45
  twistMotion=Free
 }
 {bone1Name=dyn_necklace_1, bone2Name=base_necklace, swing1LimitDegrees=10, swing2LimitDegrees=3, twistLimitDegrees=1}
 {bone1Name=dyn_necklace_2, bone2Name=dyn_necklace_1, swing1LimitDegrees=10, swing2LimitDegrees=2, twistLimitDegrees=1}
]
