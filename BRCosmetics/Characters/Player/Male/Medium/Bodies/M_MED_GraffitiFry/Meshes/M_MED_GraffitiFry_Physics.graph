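# Physics asset for the GraffitiFry medium male player body cosmetic, set up for secondary motion of the jacket.
# Six tiny base capsules link to spine_01 with free rotation and anchor limited-swing, near-zero-twist dynamic bones: single at back and sides, two-link chains at front.
# Pelvis, spine, thigh and thigh-twist helper capsules form the lower-body collision volume.
# Front-right jacket capsules are much larger than front-left ones.
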PhysicsAsset /BRCosmetics/Characters/Player/Male/Medium/Bodies/M_MED_GraffitiFry/Meshes/M_MED_GraffitiFry_Physics
bodies {
 base_jacket_back_l [{shapeType=Capsule, length=0.505, radius=0.505, shapeName=base_jacket_back_l_capsule, rotation=<rot 0 -90 0>}]
 base_jacket_back_r [{shapeType=Capsule, length=0.505, radius=0.505, shapeName=base_jacket_back_r_capsule, rotation=<rot 0 -90 0>}]
 base_jacket_front_l [{shapeType=Capsule, length=0.505, radius=0.505, shapeName=base_jacket_front_l_capsule, rotation=<rot 0 -90 0>}]
 base_jacket_front_r [{shapeType=Capsule, length=0.505, radius=0.505, shapeName=base_jacket_front_r_capsule, rotation=<rot 0 -90 0>}]
 base_jacket_mid_l [{shapeType=Capsule, length=0.505, radius=0.505, shapeName=base_jacket_mid_l_capsule, rotation=<rot 0 -90 0>}]
 base_jacket_mid_r [{shapeType=Capsule, length=0.505, radius=0.505, shapeName=base_jacket_mid_r_capsule, rotation=<rot 0 -90 0>}]
 dyn_jacket_back_l_1 [{shapeType=Capsule, length=16, radius=3.8, shapeName=dyn_jacket_back_l_1_capsule, center=<8 -2.8 -4>, rotation=<rot -45 0 -5>}]
 dyn_jacket_back_r_1 [{shapeType=Capsule, length=16, radius=3.8, shapeName=dyn_jacket_back_r_1_capsule, center=<8 -2 4>, rotation=<rot 45 0 15>}]
 dyn_jacket_front_l_1 [{shapeType=Capsule, length=6, radius=4, shapeName=dyn_jacket_front_l_1_capsule, center=<4 -2.5 2.8>}]
 dyn_jacket_front_l_2 [{shapeType=Capsule, length=9, radius=4, shapeName=dyn_jacket_front_l_1_capsule, center=<7.2 -2.5 0>, rotation=<rot 0 0 -5>}]
 dyn_jacket_front_r_1 [{shapeType=Capsule, length=22, radius=5.5, shapeName=dyn_jacket_front_l_1_capsule, center=<3.5 -2 1.5>, rotation=<rot 0 0 5>}]
 dyn_jacket_front_r_2 [{shapeType=Capsule, length=22, radius=5, shapeName=dyn_jacket_front_r_2_capsule, center=<7 -2.6 1.5>, rotation=<rot 0 0 5>}]
 dyn_jacket_mid_l_1 [{shapeType=Capsule, length=14, radius=4, shapeName=dyn_jacket_mid_l_1_capsule, center=<8 -2 0>, rotation=<rot 90 0 3>}]
 dyn_jacket_mid_r_1 [{shapeType=Capsule, length=14, radius=4.5, shapeName=dyn_jacket_mid_r_1_capsule, center=<8 -2 0>, rotation=<rot 90 0 3>}]
 pelvis [{shapeType=Capsule, length=11, radius=10, shapeName=spine_01_capsule, center=<-5 -1 0>}]
 spine_01 [{shapeType=Capsule, length=11, radius=8.8, shapeName=spine_01_capsule}]
 thigh_l [{shapeType=Capsule, length=2, radius=9, shapeName=thigh_r_capsule, center=<-1 -1.5 1.5>, rotation=<rot 0 0 90>}]
 thigh_r [{shapeType=Capsule, length=2, radius=9, shapeName=thigh_r_capsule, center=<1 1.5 -1.5>, rotation=<rot 0 0 90>}]
 thigh_twist_01_l_collision_helper [{shapeType=Capsule, length=33, radius=8.5, shapeName=thigh_r_capsule, center=<1.5 0 1.5>, rotation=<rot 90 0 0>}]
 thigh_twist_01_r_collision_helper [{shapeType=Capsule, length=33, radius=8.5, shapeName=thigh_r_capsule, center=<-1.5 0 -1.5>, rotation=<rot 90 88.3838 88.3838>}]
}
constraints [
 {
  bone1Name=spine_01
  bone2Name=pelvis
  swing1LimitDegrees=45
  swing1Motion=Free
  swing2LimitDegrees=45
  swing2Motion=Free
  twistLimitDegrees=45
  twistMotion=Free
 }
 {
  bone1Name=thigh_l
  bone2Name=pelvis
  swing1LimitDegrees=45
  swing1Motion=Free
  swing2LimitDegrees=45
  swing2Motion=Free
  twistLimitDegrees=45
  twistMotion=Free
 }
 {
  bone1Name=thigh_r
  bone2Name=pelvis
  swing1LimitDegrees=45
  swing1Motion=Free
  swing2LimitDegrees=45
  swing2Motion=Free
  twistLimitDegrees=45
  twistMotion=Free
 }
 {
  bone1Name=base_jacket_back_l
  bone2Name=spine_01
  swing1LimitDegrees=45
  swing1Motion=Free
  swing2LimitDegrees=45
  swing2Motion=Free
  twistLimitDegrees=45
  twistMotion=Free
 }
 {bone1Name=dyn_jacket_back_l_1, bone2Name=base_jacket_back_l, swing1LimitDegrees=20, swing2LimitDegrees=5, twistLimitDegrees=1}
 {
  bone1Name=base_jacket_back_r
  bone2Name=spine_01
  swing1LimitDegrees=45
  swing1Motion=Free
  swing2LimitDegrees=45
  swing2Motion=Free
  twistLimitDegrees=45
  twistMotion=Free
 }
 {bone1Name=dyn_jacket_back_r_1, bone2Name=base_jacket_back_r, swing1LimitDegrees=20, swing2LimitDegrees=5, twistLimitDegrees=1}
 {
  bone1Name=base_jacket_front_l
  bone2Name=spine_01
  swing1LimitDegrees=45
  swing1Motion=Free
  swing2LimitDegrees=45
  swing2Motion=Free
  twistLimitDegrees=45
  twistMotion=Free
 }
 {bone1Name=dyn_jacket_front_l_1, bone2Name=base_jacket_front_l, swing1LimitDegrees=45, swing2LimitDegrees=20, twistLimitDegrees=1}
 {
  bone1Name=base_jacket_front_r
  bone2Name=spine_01
  swing1LimitDegrees=45
  swing1Motion=Free
  swing2LimitDegrees=45
  swing2Motion=Free
  twistLimitDegrees=45
  twistMotion=Free
 }
 {bone1Name=dyn_jacket_front_r_1, bone2Name=base_jacket_front_r, swing1LimitDegrees=20, swing2LimitDegrees=20, twistLimitDegrees=1}
 {
  bone1Name=base_jacket_mid_l
  bone2Name=spine_01
  swing1LimitDegrees=45
  swing1Motion=Free
  swing2LimitDegrees=45
  swing2Motion=Free
  twistLimitDegrees=45
  twistMotion=Free
 }
 {bone1Name=dyn_jacket_mid_l_1, bone2Name=base_jacket_mid_l, swing1LimitDegrees=20, swing2LimitDegrees=5, twistLimitDegrees=1}
 {
  bone1Name=base_jacket_mid_r
  bone2Name=spine_01
  swing1LimitDegrees=45
  swing1Motion=Free
  swing2LimitDegrees=45
  swing2Motion=Free
  twistLimitDegrees=45
  twistMotion=Free
 }
 {bone1Name=dyn_jacket_mid_r_1, bone2Name=base_jacket_mid_r, swing1LimitDegrees=20, swing2LimitDegrees=5, twistLimitDegrees=1}
 {bone1Name=dyn_jacket_front_l_2, bone2Name=dyn_jacket_front_l_1, swing1LimitDegrees=60, swing2LimitDegrees=30, twistLimitDegrees=1}
 {bone1Name=dyn_jacket_front_r_2, bone2Name=dyn_jacket_front_r_1, swing1LimitDegrees=40, swing2LimitDegrees=30, twistLimitDegrees=1}
 {
  bone1Name=thigh_twist_01_r_collision_helper
  bone2Name=thigh_r
  swing1LimitDegrees=45
  swing1Motion=Free
  swing2LimitDegrees=45
  swing2Motion=Free
  twistLimitDegrees=45
  twistMotion=Free
 }
 {
  bone1Name=thigh_twist_01_l_collision_helper
  bone2Name=thigh_l
  swing1LimitDegrees=45
  swing1Motion=Free
  swing2LimitDegrees=45
  swing2Motion=Free
  twistLimitDegrees=45
  twistMotion=Free
 }
]
