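# Post-copy AnimBlueprint for the GraffitiFry character body, adding secondary motion to jacket, knives and necklace on top of the parent mesh's pose.
# Weighted constraints, AnimDynamics, Control Rig and two RigidBody sims form the normal pose, interleaved with additive jacket-bone corrections.
# When rigid bodies are disabled, output switches to that pose plus extra static jacket offsets.
# Update logic sets gravity override, LOD, stationary-twist and emote-montage flags.
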AnimBlueprint /BRCosmetics/Characters/Player/Male/Medium/Bodies/M_MED_GraffitiFry/Meshes/M_MED_GraffitiFry_AnimBP
target_skeleton=&/M_MED_GraffitiFry_Skeleton
graphs [
 AnimGraph {nodes [
   Output Pose(Result=Blend Poses by bool(BlendPose_0=Use cached pose 'RBANoff_Cache'(), BlendPose_1=Use cached pose 'NormalState_Cache'(), bActiveValue=Get ASWon() -> (ASWon:bool)) {blendType=Linear}) {layerGroup=None}
   n1 = "Transform (Modify) Bone - Bone: base_jacket_back_l"!(
    ComponentPose=Local To Component(LocalPose=Use cached pose 'NormalState_Cache'())
    bAlphaBoolEnabled:bool=True
    Translation="0,-1,0"
    Rotation="0,-10,0"
    Scale="1,1,1"
   ) -> (Pose) {
    alphaInputType=Bool
    rotationMode=BMM_Additive
    rotationSpace=BCS_BoneSpace
    translationMode=BMM_Additive
    translationSpace=BCS_BoneSpace
    alphaBoolBlend {blendInTime=0.5, blendOption=Sinusoidal, blendOutTime=1}
    boneToModify {boneName=base_jacket_back_l}
    rotation {yaw=-10}
    translation {y=-1}
   }
   n2 = "Transform (Modify) Bone - Bone: base_jacket_front_r"!(ComponentPose=@n1, bAlphaBoolEnabled:bool=True, Translation="0,0,0", Rotation="10,-15,5", Scale="1,1,1") -> (Pose) {
    alphaInputType=Bool
    rotationMode=BMM_Additive
    rotationSpace=BCS_BoneSpace
    translationMode=BMM_Additive
    translationSpace=BCS_BoneSpace
    alphaBoolBlend {blendInTime=0.5, blendOption=Sinusoidal, blendOutTime=1}
    boneToModify {boneName=base_jacket_front_r}
    rotation=<rot 10 -15 5>
   }
   n3 = "Transform (Modify) Bone - Bone: base_jacket_back_r"!(ComponentPose=@n2, bAlphaBoolEnabled:bool=True, Translation="0,-1,0", Rotation="0,-10,0", Scale="1,1,1") -> (Pose) {
    alphaInputType=Bool
    rotationMode=BMM_Additive
    rotationSpace=BCS_BoneSpace
    translationMode=BMM_Additive
    translationSpace=BCS_BoneSpace
    alphaBoolBlend {blendInTime=0.5, blendOption=Sinusoidal, blendOutTime=1}
    boneToModify {boneName=base_jacket_back_r}
    rotation {yaw=-10}
    translation {y=-1}
   }
   n4 = "Transform (Modify) Bone - Bone: base_jacket_front_l"!(ComponentPose=@n3, bAlphaBoolEnabled:bool=True, Translation="0,-2,0", Rotation="0,-15,0", Scale="1,1,1") -> (Pose) {
    alphaInputType=Bool
    rotationMode=BMM_Additive
    rotationSpace=BCS_BoneSpace
    translationMode=BMM_Additive
    translationSpace=BCS_BoneSpace
    alphaBoolBlend {blendInTime=0.5, blendOption=Sinusoidal, blendOutTime=1}
    boneToModify {boneName=base_jacket_front_l}
    rotation {yaw=-15}
    translation {y=-2}
   }
   n5 = "Transform (Modify) Bone - Bone: base_jacket_mid_l"!(ComponentPose=@n4, bAlphaBoolEnabled:bool=True, Translation="0,-1,0", Rotation="0,-10,0", Scale="1,1,1") -> (Pose) {
    alphaInputType=Bool
    rotationMode=BMM_Additive
    rotationSpace=BCS_BoneSpace
    translationMode=BMM_Additive
    translationSpace=BCS_BoneSpace
    alphaBoolBlend {blendInTime=0.5, blendOption=Sinusoidal, blendOutTime=1}
    boneToModify {boneName=base_jacket_mid_l}
    rotation {yaw=-10}
    translation {y=-1}
   }
   n6 = "Transform (Modify) Bone - Bone: base_jacket_front_r"(ComponentPose=@n5, bAlphaBoolEnabled=True, Translation="1,0,0", Rotation="0,-15,0", Scale="1,1,1") {
    alphaInputType=Bool
    rotationMode=BMM_Additive
    rotationSpace=BCS_BoneSpace
    translationMode=BMM_Additive
    translationSpace=BCS_BoneSpace
    alphaBoolBlend {blendInTime=0.5, blendOption=Sinusoidal, blendOutTime=1}
    boneToModify {boneName=base_jacket_front_r}
    rotation {yaw=-15}
    translation {x=1}
   }
   Save cached pose 'RBANoff_Cache'(Pose=Component To Local(ComponentPose=@n6)) {cacheName=RBANoff_Cache}
   n7 = "Constraint - base_leg_knife_r (PPPP)"!(
    ComponentPose=Local To Component(LocalPose=Copy Pose From Mesh() {bUseAttachedParent=true})
    Alpha=1
    ConstraintWeights_0=0.7582
    ConstraintWeights_1:float=0.2047
    ConstraintWeights_2:float=0.0264
    ConstraintWeights_3:float=0.0107
   ) -> (Pose) {
    alphaBoolBlend {blendOption=Linear}
    boneToModify {boneName=base_leg_knife_r}
    constraintSetup [
     {targetBone {boneName=thigh_twist_01_r}}
     {targetBone {boneName=thigh_r}}
     {targetBone {boneName=pelvis}}
     {targetBone {boneName=deform_groin_r}}
    ]
    constraintWeights [1,1,1,1]
   }
   n8 = "Constraint - base_elbowpad_knife_l (PPPPPPPP)"!(
    ComponentPose=@n7
    Alpha=1
    ConstraintWeights_0=0.3416
    ConstraintWeights_1:float=0.2372
    ConstraintWeights_2:float=0.1646
    ConstraintWeights_3:float=0.1543
    ConstraintWeights_4:float=0.0518
    ConstraintWeights_5:float=0.0376
    ConstraintWeights_6:float=0.0073
    ConstraintWeights_7:float=0.0056
   ) -> (Pose) {
    alphaBoolBlend {blendOption=Linear}
    boneToModify {boneName=base_elbowpad_knife_l}
    constraintSetup [
     {targetBone {boneName=upperarm_l}}
     {targetBone {boneName=lowerarm_l}}
     {targetBone {boneName=deform_elbow_in_l}}
     {targetBone {boneName=lowerarm_twist_02_l}}
     {targetBone {boneName=upperarm_twist_02_l}}
     {targetBone {boneName=deform_elbow_out_l}}
     {targetBone {boneName=lowerarm_twist_01_l}}
     {targetBone {boneName=upperarm_twist_01_l}}
    ]
    constraintWeights [1,1,1,1,1,1,1,1]
   }
   n9 = "Constraint - base_jacket_back_l_aimtarget (PPPP)"!(
    ComponentPose=@n8
    Alpha=1
    ConstraintWeights_0=0.5865
    ConstraintWeights_1:float=0.2888
    ConstraintWeights_2:float=0.1198
    ConstraintWeights_3:float=0.0049
   ) -> (Pose) {
    alphaBoolBlend {blendOption=Linear}
    boneToModify {boneName=base_jacket_back_l_aimtarget}
    constraintSetup [
     {targetBone {boneName=thigh_twist_01_l}}
     {targetBone {boneName=deform_glute_l}}
     {targetBone {boneName=pelvis}}
     {targetBone {boneName=thigh_l}}
    ]
    constraintWeights [1,1,1,1]
   }
   n10 = "Constraint - base_jacket_back_r_aimtarget (PPPPPP)"!(
    ComponentPose=@n9
    Alpha=1
    ConstraintWeights_0=0.5748
    ConstraintWeights_1:float=0.2651
    ConstraintWeights_2:float=0.126
    ConstraintWeights_3:float=0.0271
    ConstraintWeights_4:float=0.0054
    ConstraintWeights_5:float=0.0016
   ) -> (Pose) {
    alphaBoolBlend {blendOption=Linear}
    boneToModify {boneName=base_jacket_back_r_aimtarget}
    constraintSetup [
     {targetBone {boneName=thigh_twist_01_r}}
     {targetBone {boneName=deform_glute_r}}
     {targetBone {boneName=pelvis}}
     {targetBone {boneName=spine_01}}
     {targetBone {boneName=thigh_r}}
     {targetBone {boneName=thigh_twist_01_l}}
    ]
    constraintWeights [1,1,1,1,1,1]
   }
   n11 = Anim Dynamics - Bone: dyn_leg_knife_r(ComponentPose=@n10, Alpha=1, ExternalForce="0,0,0") {
    angularDampingOverride=1
    angularSpringConstant=20000
    bAngularSpring=true
    bOverrideAngularDamping=true
    bUseGravityOverride=true
    gravityScale=0.35
    simulationSpace=RootRelative
    alphaBoolBlend {blendOption=Linear}
    boundBone {boneName=dyn_leg_knife_r}
    componentLinearAccScale=<1 1 1>
    componentLinearVelScale=<1 1 1>
    gravityOverride {z=-980}
    physicsBodyDefinitions [{
      boundBone {boneName=dyn_leg_knife_r}
      boxExtents=<25 25 15>
      constraintSetup {angularLimitsMax=<0 1.5 1.5>, angularLimitsMin=<0 0 -1.5>, angularTarget=<1 0 0>}
      localJointOffset=<6 0 0>
     }]
   }
   n12 = Anim Dynamics - Bone: dyn_elbowpad_knife_l(ComponentPose=@n11, Alpha=1, ExternalForce="0,0,0") {
    angularDampingOverride=1
    angularSpringConstant=20000
    bAngularSpring=true
    bOverrideAngularDamping=true
    bUseGravityOverride=true
    gravityScale=0.35
    simulationSpace=RootRelative
    alphaBoolBlend {blendOption=Linear}
    boundBone {boneName=dyn_elbowpad_knife_l}
    componentLinearAccScale=<1 1 1>
    componentLinearVelScale=<1 1 1>
    gravityOverride {z=-980}
    physicsBodyDefinitions [{
      boundBone {boneName=dyn_elbowpad_knife_l}
      boxExtents=<20 30 30>
      constraintSetup {angularLimitsMax=<3 0.3 3>, angularLimitsMin=<-3 -1.5 -3>, angularTarget=<1 0 0>}
      localJointOffset=<4 0 0>
     }]
   }
   n13 = "Transform (Modify) Bone - Bone: base_jacket_front_r"(ComponentPose=@n12, bAlphaBoolEnabled=True, Translation="0,0,5", Rotation="30,-35,0", Scale="1,1,1") {
    alphaInputType=Bool
    rotationMode=BMM_Additive
    rotationSpace=BCS_BoneSpace
    translationMode=BMM_Additive
    translationSpace=BCS_BoneSpace
    alphaBoolBlend {blendInTime=0.5, blendOption=Sinusoidal, blendOutTime=1}
    boneToModify {boneName=base_jacket_front_r}
    rotation {pitch=30, yaw=-35}
    translation {z=5}
   }
   n14 = "Transform (Modify) Bone - Bone: base_jacket_front_r"(ComponentPose=@n13, bAlphaBoolEnabled=True, Translation="0,0,5", Rotation="0,0,0", Scale="1,1,1") {
    alphaInputType=Bool
    rotationMode=BMM_Additive
    rotationSpace=BCS_BoneSpace
    translationMode=BMM_Additive
    translationSpace=BCS_BoneSpace
    alphaBoolBlend {blendInTime=0.5, blendOption=Sinusoidal, blendOutTime=1}
    boneToModify {boneName=base_jacket_front_r}
    translation {z=5}
   }
   n15 = "Transform (Modify) Bone - Bone: base_jacket_back_l"(ComponentPose=@n14, bAlphaBoolEnabled=True, Translation="0,-1,0", Rotation="0,0,0", Scale="1,1,1") {
    alphaInputType=Bool
    rotationMode=BMM_Additive
    rotationSpace=BCS_BoneSpace
    translationMode=BMM_Additive
    translationSpace=BCS_BoneSpace
    alphaBoolBlend {blendInTime=0.5, blendOption=Sinusoidal, blendOutTime=1}
    boneToModify {boneName=base_jacket_back_l}
    translation {y=-1}
   }
   n16 = "Transform (Modify) Bone - Bone: base_jacket_front_r"(ComponentPose=@n15, bAlphaBoolEnabled=True, Translation="0,0,0", Rotation="0,-10,0", Scale="1,1,1") {
    alphaInputType=Bool
    rotationMode=BMM_Additive
    rotationSpace=BCS_BoneSpace
    translationMode=BMM_Additive
    translationSpace=BCS_BoneSpace
    alphaBoolBlend {blendInTime=0.5, blendOption=Sinusoidal, blendOutTime=1}
    boneToModify {boneName=base_jacket_front_r}
    rotation {yaw=-10}
   }
   n17 = "Transform (Modify) Bone - Bone: base_jacket_mid_r"!(ComponentPose=@n16, bAlphaBoolEnabled:bool=True, Translation="0,0,0", Rotation="-10,-10,0", Scale="1,1,1") -> (Pose) {
    alphaInputType=Bool
    rotationMode=BMM_Additive
    rotationSpace=BCS_BoneSpace
    translationMode=BMM_Additive
    translationSpace=BCS_BoneSpace
    alphaBoolBlend {blendInTime=0.5, blendOption=Sinusoidal, blendOutTime=1}
    boneToModify {boneName=base_jacket_mid_r}
    rotation {pitch=-10, yaw=-10}
   }
   n18 = "Transform (Modify) Bone - Bone: base_jacket_back_r"(ComponentPose=@n17, bAlphaBoolEnabled=True, Translation="0,-1,0", Rotation="0,0,0", Scale="1,1,1") {
    alphaInputType=Bool
    rotationMode=BMM_Additive
    rotationSpace=BCS_BoneSpace
    translationMode=BMM_Additive
    translationSpace=BCS_BoneSpace
    alphaBoolBlend {blendInTime=0.5, blendOption=Sinusoidal, blendOutTime=1}
    boneToModify {boneName=base_jacket_back_r}
    translation {y=-1}
   }
   n19 = Control Rig!(
    Source=Component To Local(ComponentPose=@n18)
    Alpha
    SkyDiving:bool=Get bIsSkyDiving()
    Emote:bool=Get bIsPlayingEmote()
    EmoteNoControlRig:bool=Get EmoteNoControlRig() -> (EmoteNoControlRig:bool)
   ) -> (Pose) {
    alphaBoolBlend {blendOption=Linear}
    controlRigAssetReference {blueprintRigClass=&/M_MED_GraffitiFry_CtrlRig.M_MED_GraffitiFry_CtrlRig_C}
   }
   n20 = "Transform (Modify) Bone - Bone: base_jacket_back_l"!(ComponentPose=Local To Component(LocalPose=@n19), Alpha:float=1, Translation="0,-2,0", Rotation="0,0,0", Scale="1,1,1") -> (Pose) {
    rotationMode=BMM_Additive
    rotationSpace=BCS_BoneSpace
    translationMode=BMM_Additive
    translationSpace=BCS_BoneSpace
    alphaBoolBlend {blendInTime=0.5, blendOption=Sinusoidal, blendOutTime=1}
    boneToModify {boneName=base_jacket_back_l}
    translation {y=-2}
   }
   n21 = "Transform (Modify) Bone - Bone: base_jacket_mid_l"!(ComponentPose=@n20, Alpha:float=1, Translation="0,-1.5,0", Rotation="0,-5,0", Scale="1,1,1") -> (Pose) {
    rotationMode=BMM_Additive
    rotationSpace=BCS_BoneSpace
    translationMode=BMM_Additive
    translationSpace=BCS_BoneSpace
    alphaBoolBlend {blendInTime=0.5, blendOption=Sinusoidal, blendOutTime=1}
    boneToModify {boneName=base_jacket_mid_l}
    rotation {yaw=-5}
    translation {y=-1.5}
   }
   n22 = "Transform (Modify) Bone - Bone: base_jacket_front_l"!(ComponentPose=@n21, Alpha:float=1, Translation="0,-1,0", Rotation="0,-5,0", Scale="1,1,1") -> (Pose) {
    rotationMode=BMM_Additive
    rotationSpace=BCS_BoneSpace
    translationMode=BMM_Additive
    translationSpace=BCS_BoneSpace
    alphaBoolBlend {blendInTime=0.5, blendOption=Sinusoidal, blendOutTime=1}
    boneToModify {boneName=base_jacket_front_l}
    rotation {yaw=-5}
    translation {y=-1}
   }
   n23 = "Transform (Modify) Bone - Bone: base_jacket_front_r"!(ComponentPose=@n22, Alpha:float=1, Translation="0,-1,-2", Rotation="0,-5,0", Scale="1,1,1") -> (Pose) {
    rotationMode=BMM_Additive
    rotationSpace=BCS_BoneSpace
    translationMode=BMM_Additive
    translationSpace=BCS_BoneSpace
    alphaBoolBlend {blendInTime=0.5, blendOption=Sinusoidal, blendOutTime=1}
    boneToModify {boneName=base_jacket_front_r}
    rotation {yaw=-5}
    translation {y=-1, z=-2}
   }
   n24 = "Transform (Modify) Bone - Bone: base_jacket_mid_r"!(ComponentPose=@n23, Alpha:float=1, Translation="0,-1.5,0", Rotation="0,-5,0", Scale="1,1,1") -> (Pose) {
    rotationMode=BMM_Additive
    rotationSpace=BCS_BoneSpace
    translationMode=BMM_Additive
    translationSpace=BCS_BoneSpace
    alphaBoolBlend {blendInTime=0.5, blendOption=Sinusoidal, blendOutTime=1}
    boneToModify {boneName=base_jacket_mid_r}
    rotation {yaw=-5}
    translation {y=-1.5}
   }
   n25 = "Transform (Modify) Bone - Bone: base_jacket_back_r"!(ComponentPose=@n24, Alpha:float=1, Translation="0,-2,0", Rotation="0,0,0", Scale="1,1,1") -> (Pose) {
    rotationMode=BMM_Additive
    rotationSpace=BCS_BoneSpace
    translationMode=BMM_Additive
    translationSpace=BCS_BoneSpace
    alphaBoolBlend {blendInTime=0.5, blendOption=Sinusoidal, blendOutTime=1}
    boneToModify {boneName=base_jacket_back_r}
    translation {y=-2}
   }
   n26 = Get GravityOverride()
   n27 = RigidBody!(ComponentPose=@n25, Alpha, OverrideWorldGravity:Vector=@n26, ExternalForce) -> (Pose) {
    bClampLinearTranslationLimitToRefPose=true
    bOverrideWorldGravity=true
    overridePhysicsAsset=&/M_MED_GraffitiFry_Physics
    simulationSpace=BaseBoneSpace
    alphaBoolBlend {blendOption=Linear}
    baseBoneRef {boneName=pelvis}
    componentAppliedLinearAccClamp=<3000 3000 3000>
    componentLinearAccScale=<0.6 0.6 0.6>
    componentLinearVelScale=<0.6 0.6 0.6>
    simSpaceSettings {
     maxAngularAcceleration=1000
     maxAngularVelocity=3
     maxLinearAcceleration=700
     maxLinearVelocity=500
     velocityScaleZ=0
     worldAlpha=0.2
     externalLinearDragV=<0.2 0.2>
    }
   }
   n28 = RigidBody(ComponentPose=@n27, OverrideWorldGravity=@n26) {
    bClampLinearTranslationLimitToRefPose=true
    bOverrideWorldGravity=true
    overridePhysicsAsset=&/M_MED_GraffitiFry_Necklace_Physics
    simulationSpace=BaseBoneSpace
    alphaBoolBlend {blendOption=Linear}
    baseBoneRef {boneName=spine_05}
    componentAppliedLinearAccClamp=<3000 3000 3000>
    componentLinearAccScale=<0.6 0.6 0.6>
    componentLinearVelScale=<0.6 0.6 0.6>
    simSpaceSettings {
     maxAngularAcceleration=1000
     maxAngularVelocity=3
     maxLinearAcceleration=700
     maxLinearVelocity=500
     velocityScaleZ=0
     worldAlpha=0.2
     externalLinearDragV=<0.2 0.2>
    }
   }
   n29 = "Transform (Modify) Bone - Bone: base_jacket_front_l"!(ComponentPose=@n28, bAlphaBoolEnabled:bool=True, Translation="-2,-2,0", Rotation="0,0,0", Scale="1,1,1") -> (Pose) {
    alphaInputType=Bool
    rotationMode=BMM_Additive
    rotationSpace=BCS_BoneSpace
    translationMode=BMM_Additive
    translationSpace=BCS_BoneSpace
    alphaBoolBlend {blendInTime=0.5, blendOption=Sinusoidal, blendOutTime=1}
    boneToModify {boneName=base_jacket_front_l}
    translation=<-2 -2>
   }
   n30 = "Transform (Modify) Bone - Bone: base_jacket_front_r"!(ComponentPose=@n29, bAlphaBoolEnabled:bool=True, Translation="-2,-2,0", Rotation="5,-5,0", Scale="1,1,1") -> (Pose) {
    alphaInputType=Bool
    rotationMode=BMM_Additive
    rotationSpace=BCS_BoneSpace
    translationMode=BMM_Additive
    translationSpace=BCS_BoneSpace
    alphaBoolBlend {blendOption=Sinusoidal, blendOutTime=1}
    boneToModify {boneName=base_jacket_front_r}
    rotation {pitch=5, yaw=-5}
    translation=<-2 -2>
   }
   n31 = "Transform (Modify) Bone - Bone: base_jacket_mid_r"!(ComponentPose=@n30, bAlphaBoolEnabled:bool=True, Translation="0,-1,0", Rotation="0,-5,0", Scale="1,1,1") -> (Pose) {
    alphaInputType=Bool
    rotationMode=BMM_Additive
    rotationSpace=BCS_BoneSpace
    translationMode=BMM_Additive
    translationSpace=BCS_BoneSpace
    alphaBoolBlend {blendOption=Sinusoidal, blendOutTime=1}
    boneToModify {boneName=base_jacket_mid_r}
    rotation {yaw=-5}
    translation {y=-1}
   }
   n32 = "Transform (Modify) Bone - Bone: base_jacket_front_l"(ComponentPose=@n31, bAlphaBoolEnabled=True, Translation="0,-2,0", Rotation="0,-11,0", Scale="1,1,1") {
    alphaInputType=Bool
    rotationMode=BMM_Additive
    rotationSpace=BCS_BoneSpace
    translationMode=BMM_Additive
    translationSpace=BCS_BoneSpace
    alphaBoolBlend {blendInTime=0.5, blendOption=Sinusoidal, blendOutTime=1}
    boneToModify {boneName=base_jacket_front_l}
    rotation {yaw=-11}
    translation {y=-2}
   }
   n33 = "Transform (Modify) Bone - Bone: base_jacket_front_l"(ComponentPose=@n32, bAlphaBoolEnabled=True, Translation="0,0,0", Rotation="0,-10,0", Scale="1,1,1") {
    alphaInputType=Bool
    rotationMode=BMM_Additive
    rotationSpace=BCS_BoneSpace
    translationMode=BMM_Additive
    translationSpace=BCS_BoneSpace
    alphaBoolBlend {blendInTime=0.5, blendOption=Sinusoidal, blendOutTime=1}
    boneToModify {boneName=base_jacket_front_l}
    rotation {yaw=-10}
   }
   n34 = "Transform (Modify) Bone - Bone: base_jacket_front_l"(ComponentPose=@n33, bAlphaBoolEnabled=True, Translation="0,0,0", Rotation="0,-10,0", Scale="1,1,1") {
    alphaInputType=Bool
    rotationMode=BMM_Additive
    rotationSpace=BCS_BoneSpace
    translationMode=BMM_Additive
    translationSpace=BCS_BoneSpace
    alphaBoolBlend {blendInTime=0.5, blendOption=Sinusoidal, blendOutTime=1}
    boneToModify {boneName=base_jacket_front_l}
    rotation {yaw=-10}
   }
   n35 = "Transform (Modify) Bone - Bone: base_jacket_front_r"(ComponentPose=@n34, bAlphaBoolEnabled=True, Translation="-2,0,0", Rotation="10,-10,20", Scale="1,1,1") {
    alphaInputType=Bool
    rotationMode=BMM_Additive
    rotationSpace=BCS_BoneSpace
    translationMode=BMM_Additive
    translationSpace=BCS_BoneSpace
    alphaBoolBlend {blendInTime=0.5, blendOption=Sinusoidal, blendOutTime=1}
    boneToModify {boneName=base_jacket_front_r}
    rotation=<rot 10 -10 20>
    translation {x=-2}
   }
   n36 = "Transform (Modify) Bone - Bone: base_jacket_mid_r"(ComponentPose=@n35, bAlphaBoolEnabled=True, Translation="0,0,0", Rotation="0,-10,0", Scale="1,1,1") {
    alphaInputType=Bool
    rotationMode=BMM_Additive
    rotationSpace=BCS_BoneSpace
    translationMode=BMM_Additive
    translationSpace=BCS_BoneSpace
    alphaBoolBlend {blendInTime=0.5, blendOption=Sinusoidal, blendOutTime=1}
    boneToModify {boneName=base_jacket_mid_r}
    rotation {yaw=-10}
   }
   n37 = "Transform (Modify) Bone - Bone: base_jacket_mid_r"(ComponentPose=@n36, bAlphaBoolEnabled=True, Translation="0,0,0", Rotation="0,0,-40", Scale="1,1,1") {
    alphaInputType=Bool
    rotationMode=BMM_Additive
    rotationSpace=BCS_BoneSpace
    scaleMode=BMM_Additive
    scaleSpace=BCS_BoneSpace
    translationMode=BMM_Additive
    translationSpace=BCS_BoneSpace
    alphaBoolBlend {blendInTime=0.1, blendOption=Sinusoidal, blendOutTime=0.5}
    boneToModify {boneName=base_jacket_mid_r}
    rotation {roll=-40}
   }
   n38 = "Transform (Modify) Bone - Bone: base_jacket_front_r"(ComponentPose=@n37, bAlphaBoolEnabled=True, Translation="0,-6,0", Rotation="15,-30,15", Scale="1,1,1") {
    alphaInputType=Bool
    rotationMode=BMM_Additive
    rotationSpace=BCS_BoneSpace
    translationMode=BMM_Additive
    translationSpace=BCS_BoneSpace
    alphaBoolBlend {blendInTime=0.5, blendOption=Sinusoidal, blendOutTime=1}
    boneToModify {boneName=base_jacket_front_r}
    rotation=<rot 15 -30 15>
    translation {y=-6}
   }
   Save cached pose 'NormalState_Cache'(Pose=Component To Local(ComponentPose=@n38)) {cacheName=NormalState_Cache}
  ]}
 EventGraph {nodes [
   Event BlueprintUpdateAnimation()
   > n1 = Sequence!(execute) -> (then_0, then_1, then_2:Exec, then_3:Exec, then_4:Exec, then_5:Exec, then_6:Exec)
   n2 = Get bIsMoving2D()
   n3 = GetAlphaFromReferenceTransformDelta(SkeletalMeshComponent=Get Skel Mesh(), BoneName=spine_01, BaseName=root, DeltaComponentType=EulerX, InputRange=(X=0,Y=10), TwistAxis="1,0,0")
   Set RotWithoutMove_L(execute=@n1.then_4, RotWithoutMove_L:bool=AND Boolean(A="float > float"(A=@n3, B=0.95), B=NOT Boolean(A=@n2))) -> (Output_Get:bool)
   n4 = Get bIsMoving2D()
   n5 = GetAlphaFromReferenceTransformDelta(
    SkeletalMeshComponent=Get Skel Mesh()
    BoneName=spine_01
    BaseName=root
    DeltaComponentType=EulerX
    InputRange=(X=0,Y=-55)
    TwistAxis="1,0,0"
   )
   Set RotWithoutMove_R(execute=@n1.then_3, RotWithoutMove_R:bool=AND Boolean(A="float > float"(A=@n5, B=0.8), B=NOT Boolean(A=@n4))) -> (Output_Get:bool)
   Set LOD1(execute=@n1.then_2, LOD1:bool="integer >= integer"(A=Get CurrentLOD(), B=1)) -> (Output_Get:bool)
   Set ASWon(execute=@n1.then_1, ASWon:bool=NOT Boolean(A=Get bRigidBodyEnabled())) -> (Output_Get:bool)
   n6 = GetAnimInstance(self=GetOwnerSkeletalMeshComponent())
   n7 = OR Boolean!(
    A=Montage_IsPlaying(self=@n6, Montage=/BRCosmetics/Animation/Game/MainPlayer/Emotes/TrophyCelebration_FNCS/Emote_TrophyCelebration_FNCS_CMM_M)
    B=Montage_IsPlaying(self=@n6, Montage=/BRCosmetics/Animation/Game/MainPlayer/Emotes/TruckerHorn/Emote_TruckerHorn_CMM_M)
    C:bool=Montage_IsPlaying(self=@n6, Montage=/BRCosmetics/Animation/Game/MainPlayer/Emotes/Grasshopper/Grasshopper_CMM_M)
   ) -> (ReturnValue)
   n8 = Branch(execute=@n1.then_5, Condition=@n7)
   > Set EmoteNoControlRig(EmoteNoControlRig:bool=@n7) -> (Output_Get:bool)
   Set EmoteNoControlRig(execute=@n8.else, EmoteNoControlRig=@n7)
   > Set GladiatorEmote(GladiatorEmote:bool=Montage_IsPlaying(self=@n6, Montage=/BRCosmetics/Animation/Game/MainPlayer/Emotes/Ancient_Gladiator/Emote_Ancient_Gladiator_CMM_M)) -> (Output_Get:bool)
   n9 = Gravity Override State Function DEPRECATED(
    execute=@n1.then_0
    CharacterPartAnimInstance=Self-Reference()
    GameTimeInSeconds=GetGameTimeInSeconds()
    GravityOverrideParams=Get Gravity Override Params()
   )
   > Set GravityOverride(GravityOverride=@n9)
   Event BlueprintInitializeAnimation()
   > Set Skel Mesh(Skel Mesh=GetOwnerSkeletalMeshComponent())
  ]}
]
variables [
 ASWon {type=boolean}
 EmoteNoControlRig {type=boolean}
 GladiatorEmote {type=boolean}
 Gravity Override Params {type=GravityOverrideParamsStruct}
 GravityOverride {type=Vector}
 LOD1 {type=boolean}
 RotWithoutMove_L {type=boolean}
 RotWithoutMove_R {type=boolean}
 Skel Mesh {type=SkeletalMeshComponent}
]
